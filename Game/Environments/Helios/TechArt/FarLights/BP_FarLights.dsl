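; FarLights actor that renders distant light flares and ground glows as one instanced static mesh.
; Set_InstancesFromWorldActors caches transforms from seeded random downward traces or from actors of listed classes within MaxRadius, with optional sky check and distance-lerped scale.
; The construction script rebuilds instances from that cache plus manual locations.
; Per-class overrides adjust offset and scale, and pass colors and flare scale to the material through per-instance custom data.
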(event EventBeginPlay)

(event Collision|EventActorBeginOverlap
  (params (OtherActor :type Actor%20Object%20Reference)))

(event EventTick
  (params (DeltaSeconds :type Float%20%28single-precision%29)))

(fn AddInstances
  (bind _instancedstaticmesh (Variables|Default|GetInstancedStaticMesh))
  (bind _overridelookup (Variables|DontEdit|GetOverrideLookup))
  (bind _override_per_actor_class (Variables|Default|GetOverride%20Per%20Actor%20Class))
  (bind _instancedstaticmesh_2 (Variables|Default|GetInstancedStaticMesh))
  (Utilities|FlowControl|Sequence
    (:then_0
      (for (_array_element _array_index) (Variables|DontEdit|GetTransformList)
        (bind _returnvalue (Components|InstancedStaticMesh|AddInstance _instancedstaticmesh_2 _array_element true))
        (if (>= (Utilities|Array|Get(acopy) _overridelookup _array_index) 0)
          (Components|InstancedStaticMesh|SetCustomDataValue _instancedstaticmesh_2 _returnvalue 0 1.0 false)
          (CallFunction|SetCustomPerInstanceValueVector3 (Variables|Default|GetInstancedStaticMesh) _returnvalue (Math|Conversions|ToVector(LinearColor) (.Color_Flare1_6_3E07B3A0411BE656A97419BAB12B7E9A (Utilities|Struct|BreakStructFarLightsOverride (Utilities|Array|Get(acopy) _override_per_actor_class (Utilities|Array|Get(acopy) _overridelookup _array_index))))) 1)
          (CallFunction|SetCustomPerInstanceValueVector3 (Variables|Default|GetInstancedStaticMesh) _returnvalue (Math|Conversions|ToVector(LinearColor) (.Color_Flare2_8_E9F0772E41B6FEBCB8FCA9B8AFE68475 (Utilities|Struct|BreakStructFarLightsOverride (Utilities|Array|Get(acopy) _override_per_actor_class (Utilities|Array|Get(acopy) _overridelookup _array_index))))) 4)
          (CallFunction|SetCustomPerInstanceValueVector3 (Variables|Default|GetInstancedStaticMesh) _returnvalue (Math|Conversions|ToVector(LinearColor) (.Color_Ground1_10_1BF24FF94C3A84DAAB90599365B94F41 (Utilities|Struct|BreakStructFarLightsOverride (Utilities|Array|Get(acopy) _override_per_actor_class (Utilities|Array|Get(acopy) _overridelookup _array_index))))) 7)
          (CallFunction|SetCustomPerInstanceValueVector3 (Variables|Default|GetInstancedStaticMesh) _returnvalue (Math|Conversions|ToVector(LinearColor) (.Color_Ground2_12_C0C409124CCAABBEF83691BA5736442B (Utilities|Struct|BreakStructFarLightsOverride (Utilities|Array|Get(acopy) _override_per_actor_class (Utilities|Array|Get(acopy) _overridelookup _array_index))))) 10)
          (Components|InstancedStaticMesh|SetCustomDataValue (Variables|Default|GetInstancedStaticMesh) _returnvalue 13 (.ScaleFlare_22_CCB164C2475854641FA389BE4EB37D83 (Utilities|Struct|BreakStructFarLightsOverride (Utilities|Array|Get(acopy) _override_per_actor_class (Utilities|Array|Get(acopy) _overridelookup _array_index)))) false)
          (else
            (Components|InstancedStaticMesh|SetCustomDataValue _instancedstaticmesh_2 _returnvalue 0 0.0 false)))))
    (:then_1
      (for (_array_element_2 _array_index_2) (Variables|Default|GetManual%20Locations)
        (bind _returnvalue_2 (Components|InstancedStaticMesh|AddInstance _instancedstaticmesh (Math|Transform|MakeTransform (.location _array_element_2) (.rotation _array_element_2) "0.010000,0.010000,0.010000") false))
        (Components|InstancedStaticMesh|SetCustomDataValue _instancedstaticmesh _returnvalue_2 0 (Math|Random|RandomFloatinRangefromStream (Math|Random|MakeRandomStream _array_index_2) 0.0 1.0) true)))))

(fn Set_InstancesFromWorldActors
  (bind (_actorclass_2_5b8368aa4945f64fde0b278797de704d _color_flare1_6_3e07b3a0411be656a97419bab12b7e9a _color_flare2_8_e9f0772e41b6febcb8fca9b8afe68475 _color_ground1_10_1bf24ff94c3a84daab90599365b94f41 _color_ground2_12_c0c409124ccaabbef83691ba5736442b _offset_15_85f5dd214503618e6ed9b98609a9479a _scale_18_db771780424d157714ec6f9cdf2f462c _scale_edge_20_0a0ae5da4816046722486780cd4cc0b4) (Utilities|Struct|BreakStructFarLightsOverride (Utilities|Array|Get(acopy) (Variables|Default|GetOverride%20Per%20Actor%20Class) (LocalVariables|GetTempOverridelookup))))
  (bind _returnvalue (Transformation|GetActorLocation))
  (bind _maxradius (Variables|Default|GetMaxRadius))
  (Utilities|Array|Clear (Variables|DontEdit|GetTransformList))
  (Utilities|Array|Clear (Variables|DontEdit|GetOverrideLookup))
  (if (Variables|RandomPlacement|GetRandomPlacement)
    (Math|Random|SetRandomStreamSeed (LocalVariables|GetStream) (Variables|RandomPlacement|GetSeed))
    (for _ (range (Variables|RandomPlacement|GetNumberOfInstances))
      (bind (_outhit _returnvalue_2) (Collision|LineTraceByChannel :Start (Math|Vector|MakeVector (.x (+ (Transformation|GetActorLocation) (Math|Vector|SelectVector (* (Math|Vector|Normalize (Math|Vector|MakeVector (Math|Random|RandomFloatinRangefromStream (LocalVariables|GetStream) (* (Variables|Default|GetMaxRadius) -1.0) (Variables|Default|GetMaxRadius)) (Math|Random|RandomFloatinRangefromStream (LocalVariables|GetStream) (* (Variables|Default|GetMaxRadius) -1.0) (Variables|Default|GetMaxRadius)))) (Math|Random|RandomFloatinRangefromStream (LocalVariables|GetStream) 0.0 (Variables|Default|GetMaxRadius))) (Math|Vector|MakeVector (Math|Random|RandomFloatinRangefromStream (LocalVariables|GetStream) (* (Variables|Default|GetMaxRadius) -1.0) (Variables|Default|GetMaxRadius)) (Math|Random|RandomFloatinRangefromStream (LocalVariables|GetStream) (* (Variables|Default|GetMaxRadius) -1.0) (Variables|Default|GetMaxRadius))) (> (Math|Vector|VectorLength (Math|Vector|MakeVector (Math|Random|RandomFloatinRangefromStream (LocalVariables|GetStream) (* (Variables|Default|GetMaxRadius) -1.0) (Variables|Default|GetMaxRadius)) (Math|Random|RandomFloatinRangefromStream (LocalVariables|GetStream) (* (Variables|Default|GetMaxRadius) -1.0) (Variables|Default|GetMaxRadius)))) (Variables|Default|GetMaxRadius))))) (.y (+ (Transformation|GetActorLocation) (Math|Vector|SelectVector (* (Math|Vector|Normalize (Math|Vector|MakeVector (Math|Random|RandomFloatinRangefromStream (LocalVariables|GetStream) (* (Variables|Default|GetMaxRadius) -1.0) (Variables|Default|GetMaxRadius)) (Math|Random|RandomFloatinRangefromStream (LocalVariables|GetStream) (* (Variables|Default|GetMaxRadius) -1.0) (Variables|Default|GetMaxRadius)))) (Math|Random|RandomFloatinRangefromStream (LocalVariables|GetStream) 0.0 (Variables|Default|GetMaxRadius))) (Math|Vector|MakeVector (Math|Random|RandomFloatinRangefromStream (LocalVariables|GetStream) (* (Variables|Default|GetMaxRadius) -1.0) (Variables|Default|GetMaxRadius)) (Math|Random|RandomFloatinRangefromStream (LocalVariables|GetStream) (* (Variables|Default|GetMaxRadius) -1.0) (Variables|Default|GetMaxRadius))) (> (Math|Vector|VectorLength (Math|Vector|MakeVector (Math|Random|RandomFloatinRangefromStream (LocalVariables|GetStream) (* (Variables|Default|GetMaxRadius) -1.0) (Variables|Default|GetMaxRadius)) (Math|Random|RandomFloatinRangefromStream (LocalVariables|GetStream) (* (Variables|Default|GetMaxRadius) -1.0) (Variables|Default|GetMaxRadius)))) (Variables|Default|GetMaxRadius))))) 5e+04) :End (Math|Vector|MakeVector (.x (+ (Transformation|GetActorLocation) (Math|Vector|SelectVector (* (Math|Vector|Normalize (Math|Vector|MakeVector (Math|Random|RandomFloatinRangefromStream (LocalVariables|GetStream) (* (Variables|Default|GetMaxRadius) -1.0) (Variables|Default|GetMaxRadius)) (Math|Random|RandomFloatinRangefromStream (LocalVariables|GetStream) (* (Variables|Default|GetMaxRadius) -1.0) (Variables|Default|GetMaxRadius)))) (Math|Random|RandomFloatinRangefromStream (LocalVariables|GetStream) 0.0 (Variables|Default|GetMaxRadius))) (Math|Vector|MakeVector (Math|Random|RandomFloatinRangefromStream (LocalVariables|GetStream) (* (Variables|Default|GetMaxRadius) -1.0) (Variables|Default|GetMaxRadius)) (Math|Random|RandomFloatinRangefromStream (LocalVariables|GetStream) (* (Variables|Default|GetMaxRadius) -1.0) (Variables|Default|GetMaxRadius))) (> (Math|Vector|VectorLength (Math|Vector|MakeVector (Math|Random|RandomFloatinRangefromStream (LocalVariables|GetStream) (* (Variables|Default|GetMaxRadius) -1.0) (Variables|Default|GetMaxRadius)) (Math|Random|RandomFloatinRangefromStream (LocalVariables|GetStream) (* (Variables|Default|GetMaxRadius) -1.0) (Variables|Default|GetMaxRadius)))) (Variables|Default|GetMaxRadius))))) (.y (+ (Transformation|GetActorLocation) (Math|Vector|SelectVector (* (Math|Vector|Normalize (Math|Vector|MakeVector (Math|Random|RandomFloatinRangefromStream (LocalVariables|GetStream) (* (Variables|Default|GetMaxRadius) -1.0) (Variables|Default|GetMaxRadius)) (Math|Random|RandomFloatinRangefromStream (LocalVariables|GetStream) (* (Variables|Default|GetMaxRadius) -1.0) (Variables|Default|GetMaxRadius)))) (Math|Random|RandomFloatinRangefromStream (LocalVariables|GetStream) 0.0 (Variables|Default|GetMaxRadius))) (Math|Vector|MakeVector (Math|Random|RandomFloatinRangefromStream (LocalVariables|GetStream) (* (Variables|Default|GetMaxRadius) -1.0) (Variables|Default|GetMaxRadius)) (Math|Random|RandomFloatinRangefromStream (LocalVariables|GetStream) (* (Variables|Default|GetMaxRadius) -1.0) (Variables|Default|GetMaxRadius))) (> (Math|Vector|VectorLength (Math|Vector|MakeVector (Math|Random|RandomFloatinRangefromStream (LocalVariables|GetStream) (* (Variables|Default|GetMaxRadius) -1.0) (Variables|Default|GetMaxRadius)) (Math|Random|RandomFloatinRangefromStream (LocalVariables|GetStream) (* (Variables|Default|GetMaxRadius) -1.0) (Variables|Default|GetMaxRadius)))) (Variables|Default|GetMaxRadius))))) 2e+02) :TraceChannel "TraceTypeQuery1" :bTraceComplex false :DrawDebugType "ForOneFrame" :bIgnoreSelf true :TraceColor "(R=1.000000,G=0.000000,B=0.000000,A=1.000000)" :TraceHitColor "(R=0.000000,G=1.000000,B=0.000000,A=1.000000)" :DrawTime 2.0))
      (if _returnvalue_2
        (Utilities|Array|Add (Variables|DontEdit|GetTransformList) (Math|Transform|MakeTransform (+ (Variables|Default|GetOffset) (.Location (Collision|BreakHitResult _outhit))) "0, 0, 0" (Math|Vector|MakeVector (+ (Math|Float|Lerp (Variables|Default|GetScale) (Variables|Default|GetScale_Edge) (/ (Math|Vector|Distance(Vector) (.Location (Collision|BreakHitResult _outhit)) (Transformation|GetActorLocation)) (Variables|Default|GetMaxRadius))) (* (- (Math|Random|RandomFloatinRangefromStream (Math|Random|MakeRandomStream 0) 0.0 1.0) 0.5) (Variables|Default|GetScaleRandom))) (+ (Math|Float|Lerp (Variables|Default|GetScale) (Variables|Default|GetScale_Edge) (/ (Math|Vector|Distance(Vector) (.Location (Collision|BreakHitResult _outhit)) (Transformation|GetActorLocation)) (Variables|Default|GetMaxRadius))) (* (- (Math|Random|RandomFloatinRangefromStream (Math|Random|MakeRandomStream 0) 0.0 1.0) 0.5) (Variables|Default|GetScaleRandom))) (+ (Math|Float|Lerp (Variables|Default|GetScale) (Variables|Default|GetScale_Edge) (/ (Math|Vector|Distance(Vector) (.Location (Collision|BreakHitResult _outhit)) (Transformation|GetActorLocation)) (Variables|Default|GetMaxRadius))) (* (- (Math|Random|RandomFloatinRangefromStream (Math|Random|MakeRandomStream 0) 0.0 1.0) 0.5) (Variables|Default|GetScaleRandom))))))))
    (Components|InstancedStaticMesh|ClearInstances (Variables|Default|GetInstancedStaticMesh))
    (CallFunction|AddInstances)
    (else
      (Utilities|FlowControl|Sequence
        (:then_1
          (for _array_element (Variables|Default|GetActor%20Class)
            (bind _outactors (Actor|GetAllActorsOfClass _array_element))
            (for (_array_element_2 _array_index) _outactors
              (if (Variables|Default|GetDoSkyCheck)
                (bind (_outhit_2 _returnvalue_3) (Collision|LineTraceByChannel (+ (Transformation|GetActorLocation _array_element_2) "0.000000,0.000000,10000.000000") (Transformation|GetActorLocation _array_element_2) "TraceTypeQuery1" false (Utilities|Array|MakeArray _array_element_2) "ForOneFrame"))
                (if _returnvalue_3
                  (if (< (Math|Vector|Distance(Vector) (.Location (Collision|BreakHitResult _outhit_2)) (Transformation|GetActorLocation _array_element_2)) 2.5e+02)
                    _)))
              (if (< (Math|Vector|Distance(Vector) (Transformation|GetActorLocation _array_element_2) _returnvalue) _maxradius)
                (Utilities|FlowControl|Sequence
                  (:then_0
                    (LocalVariables|SetTempOverridelookup -1)
                    (Utilities|FlowControl|Sequence
                      (:then_0
                        (bind (_array_element_3 _array_index_2) (Utilities|Array|ForEachLoopwithBreak (Variables|Default|GetOverride%20Per%20Actor%20Class))
                          (:LoopBody
                            (Development|PrintString (Utilities|String|Append "DBUG " (Utilities|GetClassDisplayName (Utilities|Struct|BreakStructFarLightsOverride _array_element_3)) "   // CurrentClass = " (Utilities|GetClassDisplayName _array_element) " Vallue = " (Utilities|String|ToString(Boolean) (== (Utilities|Struct|BreakStructFarLightsOverride _array_element_3) _array_element))) false false)
                            (if (== (Utilities|Struct|BreakStructFarLightsOverride _array_element_3) _array_element)
                              (LocalVariables|SetTempOverridelookup _array_index_2)))))
                      (:then_1
                        (Utilities|Array|Add (Variables|DontEdit|GetOverrideLookup) (LocalVariables|GetTempOverridelookup)))))
                  (:then_1
                    (Utilities|Array|Add (Variables|DontEdit|GetTransformList) (Math|Transform|MakeTransform (+ (.location (Transformation|GetActorTransform _array_element_2)) (Math|Vector|RotateVector (Math|Vector|SelectVector _offset_15_85f5dd214503618e6ed9b98609a9479a (Variables|Default|GetOffset) (>= (LocalVariables|GetTempOverridelookup) 0)) (.rotation (Transformation|GetActorTransform _array_element_2)))) (.rotation (Transformation|GetActorTransform _array_element_2)) (Math|Vector|MakeVector (+ (Math|Float|Lerp (Math|Float|SelectFloat _scale_18_db771780424d157714ec6f9cdf2f462c (Variables|Default|GetScale) (>= (LocalVariables|GetTempOverridelookup) 0)) (Math|Float|SelectFloat _scale_edge_20_0a0ae5da4816046722486780cd4cc0b4 (Variables|Default|GetScale_Edge) (>= (LocalVariables|GetTempOverridelookup) 0)) (/ (Math|Vector|Distance(Vector) (Transformation|GetActorLocation _array_element_2) _returnvalue) _maxradius)) (* (- (Math|Random|RandomFloatinRangefromStream (Math|Random|MakeRandomStream _array_index) 0.0 1.0) 0.5) (Variables|Default|GetScaleRandom))) (+ (Math|Float|Lerp (Math|Float|SelectFloat _scale_18_db771780424d157714ec6f9cdf2f462c (Variables|Default|GetScale) (>= (LocalVariables|GetTempOverridelookup) 0)) (Math|Float|SelectFloat _scale_edge_20_0a0ae5da4816046722486780cd4cc0b4 (Variables|Default|GetScale_Edge) (>= (LocalVariables|GetTempOverridelookup) 0)) (/ (Math|Vector|Distance(Vector) (Transformation|GetActorLocation _array_element_2) _returnvalue) _maxradius)) (* (- (Math|Random|RandomFloatinRangefromStream (Math|Random|MakeRandomStream _array_index) 0.0 1.0) 0.5) (Variables|Default|GetScaleRandom))) (+ (Math|Float|Lerp (Math|Float|SelectFloat _scale_18_db771780424d157714ec6f9cdf2f462c (Variables|Default|GetScale) (>= (LocalVariables|GetTempOverridelookup) 0)) (Math|Float|SelectFloat _scale_edge_20_0a0ae5da4816046722486780cd4cc0b4 (Variables|Default|GetScale_Edge) (>= (LocalVariables|GetTempOverridelookup) 0)) (/ (Math|Vector|Distance(Vector) (Transformation|GetActorLocation _array_element_2) _returnvalue) _maxradius)) (* (- (Math|Random|RandomFloatinRangefromStream (Math|Random|MakeRandomStream _array_index) 0.0 1.0) 0.5) (Variables|Default|GetScaleRandom))))))))))
            (Components|InstancedStaticMesh|ClearInstances (Variables|Default|GetInstancedStaticMesh))
            (CallFunction|AddInstances)))))))

(fn SetCustomPerInstanceValueVector3
  (params (Instance :type Instanced%20Static%20Mesh%20Component%20Object%20Reference) (InstanceIndex :type Integer) (ValueVector3 :type Vector) (CustomDataIndex :type Integer))
  (Components|InstancedStaticMesh|SetCustomDataValue Instance InstanceIndex CustomDataIndex (.x ValueVector3) false)
  (Components|InstancedStaticMesh|SetCustomDataValue Instance InstanceIndex (+ CustomDataIndex 1) (.y ValueVector3) false)
  (Components|InstancedStaticMesh|SetCustomDataValue Instance InstanceIndex (+ CustomDataIndex 2) (.z ValueVector3) true))

(fn ConstructionScript
  (Rendering|Material|SetMaterial (Variables|Default|GetInstancedStaticMesh) 0 (Variables|Default|GetMaterial))
  (CallFunction|AddInstances))
