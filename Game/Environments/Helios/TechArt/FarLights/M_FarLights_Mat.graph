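# Unlit additive material for instanced distant lights, combining a camera-facing flare and a ground glow, split by vertex color green.
# World position offset scales the flare sprite and shifts both parts relative to the camera.
# Radial falloffs take two random-blended colors, overridable by per-instance data, with sine flicker, optional noise sparkle and eye-adaptation compensation.
# Depth fade softens intersections, while a FadeOut term combines camera height with time of day.
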Material /Game/Environments/Helios/TechArt/FarLights/M_FarLights_Mat
graph {
 properties {
  bUsedWithInstancedStaticMeshes=true
  bUsedWithNanite=true
  blendMode=BLEND_Additive
  pixelDepthOffsetMode=PDOM_Legacy
  shadingModel=MSM_Unlit
  usedShadingModels=Unlit
 }
 nodes [
  n1 = Subtract(A=CameraPositionWS(), B=WorldPosition().XYZ)
  n2 = Constant3Vector() {constant=<rgba 1 1 0 0>}
  n3 = Normalize(VectorInput=@n1)
  n4 = Constant3Vector() {constant=<rgba 0 0 1 0>}
  n5 = Multiply(A=Normalize(VectorInput=Multiply(A=@n1, B=@n2)), B=OneMinus(None=DotProduct(A=@n3, B=@n4)))
  n6 = Multiply(A=ScalarParameter() {defaultValue=2000, group=PositionsOffsets, parameterName=Offset}, B=@n5)
  Add(A=@n6, B=Multiply(A=@n4, B=ScalarParameter() {defaultValue=1993.4052, group=PositionsOffsets, parameterName=ZOffset}))
  n7 = CameraPositionWS()
  n8 = ComponentMask(None=@n7) {b=true}
  n9 = TransformPosition(None=Constant3Vector()) {transformSourceType=TRANSFORMPOSSOURCE_Instance, transformType=TRANSFORMPOSSOURCE_World}
  n10 = ComponentMask(None=@n9) {b=true}
  Add(A=@n8, B=@n10)
  n11 = VertexInterpolator(VS=PerInstanceRandom())
  n12 = Add(A=Time(), B=@n11.PS)
  n13 = ScalarParameter() {defaultValue=1.4474, group=Flicker, parameterName=FlickerSpeed2}
  n14 = Add(A=Sine(None=Multiply(A=@n12, B=@n13)))
  n15 = Add(A=@n11.PS, B=Time())
  n16 = ScalarParameter() {defaultValue=0.856, group=Flicker, parameterName=FlickerSpeed1}
  n17 = Add(A=Sine(None=Multiply(A=@n15, B=@n16)))
  n18 = LinearInterpolate(A=ScalarParameter() {defaultValue=0.45, group=Flicker, parameterName=FlickerMin}, Alpha=Multiply(A=Divide(A=@n14), B=Divide(A=@n17)))
  n19 = NamedRerouteUsage()
  n20 = VectorParameter() {group=Color, parameterName=ColorCenter, defaultValue=<rgba 2 0.432 0 1>}
  n21 = PerInstanceCustomData3Vector!(DefaultValue) {dataIndex=1}
  n22 = NamedRerouteUsage()
  n23 = VectorParameter() {group=Color, parameterName=ColorCenter2, defaultValue=<rgba 1 0.6347 0.1039 1>}
  n24 = PerInstanceCustomData3Vector() {dataIndex=4}
  n25 = NamedRerouteUsage()
  n26 = Multiply(
   A=LinearInterpolate(A=LinearInterpolate(A=@n20.RGB, B=@n21, Alpha=@n22), B=LinearInterpolate(A=@n23.RGB, B=@n24, Alpha=@n25), Alpha=@n11.PS)
   B=ScalarParameter() {defaultValue=0.2, group=Color, parameterName=BrightnessFlare}
  )
  n27 = Distance(A=TextureCoordinate(), B=Constant2Vector() {g=0.5, r=0.5})
  n28 = Saturate(None=OneMinus(None=Divide(A=@n27) {constB=0.5}))
  n29 = ScalarParameter() {defaultValue=7, group=FalloffSettings, parameterName=FalloffCenter_FadeOut}
  n30 = ScalarParameter() {defaultValue=11.0446, group=FalloffSettings, parameterName=FalloffCenter}
  n31 = NamedRerouteUsage()
  n32 = ScalarParameter() {defaultValue=2.5273, group=FalloffSettings, parameterName=FalloffMedium}
  n33 = Add(
   A=Power(Base=@n28, Exp=LinearInterpolate(A=@n29, B=@n30, Alpha=@n31))
   B=Multiply(A=Power(Base=@n28, Exp=@n32), B=ScalarParameter() {defaultValue=0.0746, group=FalloffSettings, parameterName=BirghtnessMedium})
  )
  n34 = VectorParameter() {group=Color, parameterName=Color, defaultValue=<rgba 0.18 0.1209 0.0146 1>}
  n35 = PerInstanceCustomData3Vector() {dataIndex=7}
  n36 = NamedRerouteUsage()
  n37 = VectorParameter() {group=Color, parameterName=Color2, defaultValue=<rgba 0.2083 0.1084 0.004 1>}
  n38 = PerInstanceCustomData3Vector() {dataIndex=10}
  n39 = NamedRerouteUsage()
  n40 = Multiply(
   A=LinearInterpolate(A=LinearInterpolate(A=@n34.RGB, B=@n35, Alpha=@n36), B=LinearInterpolate(A=@n37.RGB, B=@n38, Alpha=@n39), Alpha=@n11.PS)
   B=ScalarParameter() {defaultValue=0.5, group=Color, parameterName=BrightnessGround}
  )
  n41 = Power(Base=@n28, Exp=ScalarParameter() {defaultValue=2.0284, group=FalloffSettings, parameterName=Falloff})
  n42 = VertexColor()
  n43 = Multiply(A=LinearInterpolate(A=Multiply(A=@n26, B=@n33), B=Multiply(A=@n40, B=@n41), Alpha=@n42.G), B=@n19)
  Divide(
   A=Multiply(A=Multiply(A=@n18, B=@n43), B=ScalarParameter() {defaultValue=1.4399, group=Exposure, parameterName=Exposure})
   B=EyeAdaptation().EyeAdaptation
  )
  n44 = LinearInterpolate(
   A=ScalarParameter() {defaultValue=2000, parameterName=Size}
   B=PerInstanceCustomData!(DefaultValue) {constDefaultValue=2000, dataIndex=13}
   Alpha=NamedRerouteUsage()
  )
  n45 = Divide(A=PerInstanceRandom(), B=ScalarParameter() {defaultValue=3.7072, parameterName=SparkleFrequency})
  n46 = ScalarParameter() {defaultValue=29.0646, parameterName=SparkleSpeed}
  n47 = Add(A=@n45, B=Multiply(A=Time(), B=Divide(A=@n46) {constB=100}))
  n48 = Saturate(None=LinearInterpolate(
    A=ScalarParameter() {defaultValue=-1.1915, parameterName=Sparkling_low}
    B=ScalarParameter() {defaultValue=1, parameterName=Sparkling_high}
    Alpha=TextureSample(UVs=@n47) {samplerType=SAMPLERTYPE_Masks, texture=/Game/Global/Textures/Default/Noise/WindNoise_Mask_1}.R
   ))
  n49 = MaterialFunctionCall!(
   Center Location
   0-1 UVs
   XY scale=StaticSwitchParameter(True=Multiply(A=@n48, B=@n44), False=@n44) {parameterName=UseSparkling}
   OverridePivot
   "Normals (Optional)"
  ) {materialFunction=&Sprite}
  n50 = ScalarParameter() {defaultValue=1500, group=PositionsOffsets, parameterName=OffsetFlare}
  n51 = LinearInterpolate(A=Add(A=@n49.World position offset, B=Multiply(A=@n50, B=@n5)), B=@n6, Alpha=VertexColor().G)
  n52 = Multiply(A=@n3, B=ScalarParameter() {defaultValue=10, group=PositionsOffsets, parameterName=CameraDirectionOffset})
  LinearInterpolate(
   A=Subtract(A=ObjectPositionWS(), B=WorldPosition().XYZ)
   B=StaticSwitchParameter(True=@n49.World position offset, False=Add(A=@n51, B=@n52)) {parameterName=UseSparkling}
   Alpha=Ceil(None=NamedRerouteUsage())
  )
  n53 = Constant() {r=1}
  n54 = DepthFade(Opacity=@n53, FadeDistance=ScalarParameter() {defaultValue=453.4101, group=FalloffSettings, parameterName=DepthFade})
  Multiply(A=StaticSwitchParameter(True=@n54, False=LinearInterpolate(A=@n53, B=@n54, Alpha=@n42.G)) {parameterName=UseDepthFade For both}, B=@n19)
  n55 = Subtract(A=@n8, B=@n10)
  n56 = VertexColor()
  n57 = LinearInterpolate(
   A=ScalarParameter() {defaultValue=-14428.2754, group=FadeOut, parameterName=CameraHeightOffset_Flare}
   B=ScalarParameter() {defaultValue=-9632.6338, group=FadeOut, parameterName=CameraHeightOffset}
   Alpha=@n56.G
  )
  n58 = ScalarParameter() {defaultValue=49642.7188, group=FadeOut, parameterName=CameraHeightFade_Flare}
  n59 = ScalarParameter() {defaultValue=10000, group=FadeOut, parameterName=CameraHeightFade}
  n60 = Saturate(None=Divide(A=Add(A=@n55, B=@n57), B=LinearInterpolate(A=@n58, B=@n59, Alpha=@n56.G)))
  n61 = ScalarParameter() {defaultValue=0.7512, group=FadeOut, parameterName=Morning_Evening}
  n62 = ScalarParameter() {defaultValue=0.512, group=FadeOut, parameterName=Day}
  n63 = ScalarParameter() {defaultValue=1, group=FadeOut, parameterName=Night}
  NamedRerouteDeclaration(None=Multiply(
    A=@n60
    B=MaterialFunctionCall(Value_Morning=@n61, Value_Day=@n62, Value_Evening=@n61, Value_Night=@n63) {materialFunction=&Scalar_Time_of_Day}.Result
   )) {name=FadeOut, nodeColor=<rgba 1 0.3984 0 1>}
  n64 = PerInstanceCustomData(DefaultValue=Constant())
  NamedRerouteDeclaration(None=Add(A=ScalarParameter() {bUseCustomPrimitiveData=true, parameterName=PrimDataUseInstance}, B=@n64)) {name=UseInstanceData, nodeColor=<rgba 0.6172 0 1 1>}
  Reroute(None=@n64)
  n65 = Divide(
   A=Subtract(A=Distance(A=@n7, B=@n9), B=ScalarParameter() {defaultValue=30000, group=FadeOut, parameterName=DistanceHorizontal_Offset})
   B=ScalarParameter() {defaultValue=20000, group=FadeOut, parameterName=DistanceHorizontal_FadeNear}
  )
  Saturate(None=Add(A=@n60, B=Saturate(None=@n65)))
  ScalarParameter() {parameterName=este}
  StaticSwitchParameter(True=Multiply(A=@n48, B=@n43), False=@n43) {parameterName=UseSparkling}
  VertexInterpolator(VS=@n45)
 ]
}
