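; Static mesh actor for PersistentFX, named as a custom flare variant, that carries a replicated mesh and material overrides.
; setStaticMesh stores both values on the authority, with an editor/PIE check around the immediate mesh update.
; OnRep handlers apply changes wherever cosmetic events can execute.
; validateMeshMats assigns valid overrides to existing material slots; the editor-only spawn function and events are empty stubs.
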(event EventBeginPlay)

(event Collision|EventActorBeginOverlap
  (params (OtherActor :type Actor%20Object%20Reference)))

(event EventTick
  (params (DeltaSeconds :type Float%20%28single-precision%29)))

(fn editorOnly_SpawnFXActor
  (params (actorClass :type Actor%20Class%20Reference) (fxComp :type Persistent%20FXComponent%20Object%20Reference))
  (return 0))

(fn OnRep_overrideMats
  (Networking|CanExecuteCosmeticEvents
    (:True
      (CallFunction|ValidateMeshMats))))

(fn OnRep_StaticMeshToLoad
  (Networking|CanExecuteCosmeticEvents
    (:True
      (Components|StaticMesh|SetStaticMesh (Variables|StaticMeshActor|GetStaticMeshComponent) (Variables|Default|GetStaticMeshToLoad))
      (CallFunction|ValidateMeshMats))))

(fn setStaticMesh
  (params (StaticMeshToLoad :type Static%20Mesh%20Object%20Reference) (overrideMats :type Array%20of%20Material%20Interface%20Object%20References))
  (Networking|SwitchHasAuthority
    (:Authority
      (Variables|Default|SetStaticMeshToLoad StaticMeshToLoad)
      (Variables|Default|SetoverrideMats overrideMats)
      (if (Editor|InEditororPIE)
        (Components|StaticMesh|SetStaticMesh (Variables|StaticMeshActor|GetStaticMeshComponent) (Variables|Default|GetStaticMeshToLoad))
        (CallFunction|ValidateMeshMats)))))

(fn ConstructionScript)

(fn validateMeshMats
  (Utilities|IsValid (Variables|/Script/Engine.StaticMeshComponent|GetStaticMesh (Variables|StaticMeshActor|GetStaticMeshComponent))
    (:Is%20Valid
      (for (_array_element _array_index) (Variables|Default|GetoverrideMats)
        (Utilities|IsValid _array_element
          (:Is%20Valid
            (if (< _array_index (Rendering|Material|GetNumMaterials (Variables|StaticMeshActor|GetStaticMeshComponent)))
              (Rendering|Material|SetMaterial (Variables|StaticMeshActor|GetStaticMeshComponent) _array_index _array_element))))))))
